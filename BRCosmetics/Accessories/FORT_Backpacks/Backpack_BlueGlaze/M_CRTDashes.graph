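# Unlit additive material for Niagara sprites that draws a dashes icon, apparently for the BlueGlaze backpack's CRT effect.
# The grayscale dash texture is multiplied by particle color, and separately by particle alpha.
# Cloud noise, panned diagonally over UVs offset by a per-particle random value, is saturated and multiplied into the alpha product to break it up.
# A HueShift function and a second ParticleRandom node sit unconnected in the graph.
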Material /BRCosmetics/Accessories/FORT_Backpacks/Backpack_BlueGlaze/M_CRTDashes
graph {
 properties {bUsedWithNiagaraSprites=true, blendMode=BLEND_Additive, pixelDepthOffsetMode=PDOM_Legacy, shadingModel=MSM_Unlit, usedShadingModels=Unlit}
 nodes [
  MaterialFunctionCall() {materialFunction=&HueShift}
  n1 = TextureSampleParameter2D() {
   parameterName=Tex1
   samplerType=SAMPLERTYPE_LinearGrayscale
   texture=&/T_BlueGlaze_Icon_Dashes
  }
  n2 = ParticleColor()
  Multiply(A=@n1.R, B=@n2.RGB)
  n3 = Multiply(A=Add(A=ParticleRandom(), B=TextureCoordinate())) {constB=0.2}
  n4 = Power(Base=TextureSampleParameter2D(UVs=Panner(Coordinate=@n3) {speedX=0.3, speedY=-0.2}) {parameterName=Tex2, texture=&AtmosphericCloudNoise01}.RGB) {constExponent=1}
  Multiply(A=Multiply(A=@n1.R, B=@n2.A), B=Saturate(None=Multiply(A=@n4)))
  ParticleRandom()
 ]
}
