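# Idle Niagara effect for the BlueGlaze backpack cosmetic, showing a symbol and wing meshes with glow and glitch-like sprite accents.
# Symbol and Wings follow system-level AutoRotate and LogoOffset values, which also drive the dash sprites.
# Glow flickers its sprite and light renderers through a looped-age curve.
# CRT_Yellow and CRT_Rainbow burst short-lived streaks sampled on the symbol mesh, using a scratch module that reads the Symbol emitter's particle.
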NiagaraSystem /BRCosmetics/Accessories/FORT_Backpacks/Backpack_BlueGlaze/NS_BlueGlaze_Backpack_Idle
dependencies {
 usedDataInterfaces [
  &NiagaraDataInterfaceCurve
  &NiagaraDataInterfaceVector2DCurve
  &NiagaraDataInterfaceParticleRead
  &NiagaraDataInterfaceColorCurve
 ]
 usedDynamicInputs [
  &MultiplyQuaternion
  &MakeQuatFromAxisAngle
  &Lerp_Vector
  &Sine
  &Add_Float
  &Lerp_Float
  &MakeVector
  &Subtract_Float
 ]
 usedModules [
  &self:SystemSpawnScript.SystemScriptSource.SystemScriptGraph.NiagaraNodeAssignment_0.SetVariables_#1
  &SystemState
  &self:SystemSpawnScript.SystemScriptSource.SystemScriptGraph.NiagaraNodeAssignment_1.SetVariables_#2
  &self:SystemSpawnScript.SystemScriptSource.SystemScriptGraph.NiagaraNodeAssignment_2.SetVariables_#3
 ]
}
emitters {
 CRT_Rainbow {
  emitterUpdateScript [
   &EmitterState(
    Inactive Response=::NewEnumerator1
    Life Cycle Mode=::NewEnumerator1
    Loop Duration=&RandomRangeFloat(Minimum=1, Maximum=2)
    UseLoopDelay=false
   )
   &SpawnBurst_Instantaneous(
    Spawn Count=&RandomRangeInt(Minimum=10, Maximum=30)
    Use Spawn Probability=false
    UseLoopCountLimit=false
   )
  ]
  particleSpawnScript [
   &InitializeParticle(
    Color=<rgba 1 1 1 0.8>
    Lifetime=&RandomRangeFloat(Minimum=0.1, Maximum=0.2)
    Sprite Size=&RandomRangeVector2D(Minimum=<0.8 2>, Maximum=<1.2 7>)
    Sprite Size Mode=::NewEnumerator4
    UsePositionOffset=false
    Write Material Random=true
   )
   &AddVelocity(
    Velocity=&RotateVectorByQuaternion(Vector To Rotate=<55 -2 100>, Quaternion=$System.AutoRotate)
    Velocity Speed Scale=&RandomRangeFloat(Minimum=0.2, Maximum=-0.2)
   )
  ]
  particleUpdateScript [
   /Niagara/Modules/Spawn/Location/StaticMeshLocation(
    MaskSpawnGroup=false
    OffsetAlongNormal=false
    OverrideRandomSeed=false
    SamplePosOffset=false
    Static Mesh=DataInterface() {propertyValues {
      DefaultMesh=&/SM_BlueGlaze_Symbol
      InstanceIndex=-1
      LODIndex=0
      PreviewMesh=null
      SoftSourceActor=null
      SourceMode=DefaultMeshOnly
      bAllowSamplingFromStreamingLODs=false
      bCaptureTransformsPerFrame=true
      bUsePhysicsBodyVelocity=false
      FilteredSockets []
      LODIndexUserParameter {parameter {name=None, typeDefHandle {registeredTypeIndex=108}, varData []}}
      MeshParameterBinding {parameter {name=None, typeDefHandle {registeredTypeIndex=1}, varData []}}
      SectionFilter {allowedMaterialSlots []}
     }}
   )
   &self:ScratchModule(New Particle Attribute Reader=DataInterface() {propertyValues {EmitterBinding {bindingMode=Other, emitterName=Symbol}}}) {module=ScratchModule}
   &ParticleState()
   &ScaleSpriteSize(
    Non-Uniform Curve Scale=&RandomRangeVector2D(Minimum=<1 0.5>, Maximum=<1 1.5>)
    Non-Uniform Curve Sprite Scale=DataInterface() {propertyValues {
      XCurve {keys [{tangentMode=RCTM_User, time=0, value=1}, {tangentMode=RCTM_User, time=1, value=1}]}
      YCurve {keys [
        <at 0 0.5 RCIM_Cubic>
        {arriveTangent=1, interpMode=RCIM_Cubic, leaveTangent=1, time=0.9, value=2}
        {arriveTangent=-1.3057, interpMode=RCIM_Cubic, leaveTangent=-1.3057, tangentMode=RCTM_User, time=1, value=1.5}
       ]}
     }}
    Scale Sprite Size Mode=::NewEnumerator3
   )
   &ScaleColor(
    Scale Alpha=&FloatFromCurve(FloatCurve=DataInterface() {propertyValues {Curve {keys [<at 0 0>,<at 0.1 1>,<at 0.9 1>,<at 1 0>]}}})
    Scale RGB=&MakeVectorFromLinearColorRGB(Color=&ColorFromCurve(ColorCurve=DataInterface() {propertyValues {
        AlphaCurve {keys [<at 0 1 RCIM_Cubic>]}
        BlueCurve {keys [<at 0.3315 0 RCIM_Cubic>,<at 0.4437 0.1767>,<at 0.5446 1>,<at 0.7704 1>,<at 0.8489 1>]}
        GreenCurve {keys [<at 0.3315 0.8745 RCIM_Cubic>,<at 0.4437 1>,<at 0.5446 0.9304>,<at 0.7704 0.0412>,<at 0.8489 0>]}
        RedCurve {keys [<at 0.3315 1 RCIM_Cubic>,<at 0.4437 0>,<at 0.5446 0>,<at 0.7704 0>,<at 0.8489 0.7503>]}
       }}))
    ScaleA=true
    ScaleRGB=true
   )
   &SolveForcesAndVelocity(Clamp Velocity=false, Limit Acceleration=false)
  ]
  rendererClasses [&NiagaraSpriteRendererProperties]
  renderers [{rendererClass=&NiagaraSpriteRendererProperties}]
 }
 CRT_Yellow {
  emitterUpdateScript [
   &EmitterState(
    Inactive Response=::NewEnumerator1
    Life Cycle Mode=::NewEnumerator1
    Loop Duration=&RandomRangeFloat(Minimum=0.5, Maximum=2)
    UseLoopDelay=false
   )
   &SpawnBurst_Instantaneous(
    Spawn Count=&RandomRangeInt(Minimum=6, Maximum=25)
    Use Spawn Probability=false
    UseLoopCountLimit=false
   )
  ]
  particleSpawnScript [
   &InitializeParticle(
    Color=<rgba 1.5 1.3283 0 0.2>
    Lifetime=&RandomRangeFloat(Minimum=0.05, Maximum=0.2)
    Sprite Size=&RandomRangeVector2D(Minimum=<0.45 1>, Maximum=<0.9 8>)
    Sprite Size Mode=::NewEnumerator4
    UsePositionOffset=false
    Write Material Random=true
   )
   &AddVelocity(
    Velocity=&RotateVectorByQuaternion(Vector To Rotate=<55 -2 100>, Quaternion=$System.AutoRotate)
    Velocity Speed Scale=&RandomRangeFloat(Minimum=0.2, Maximum=-0.2)
   )
  ]
  particleUpdateScript [
   &StaticMeshLocation(
    MaskSpawnGroup=false
    Mesh Sampling Type:ENiagara_StaticSamplingMode=::NewEnumerator2
    OffsetAlongNormal=false
    OverrideRandomSeed=false
    SamplePosOffset=false
    Static Mesh=DataInterface() {propertyValues {
      DefaultMesh=&SM_BlueGlaze_Symbol
      InstanceIndex=-1
      LODIndex=0
      PreviewMesh=null
      SoftSourceActor=null
      SourceMode=DefaultMeshOnly
      bAllowSamplingFromStreamingLODs=false
      bCaptureTransformsPerFrame=true
      bUsePhysicsBodyVelocity=false
      FilteredSockets []
      LODIndexUserParameter {parameter {name=None, typeDefHandle {registeredTypeIndex=108}, varData []}}
      MeshParameterBinding {parameter {name=None, typeDefHandle {registeredTypeIndex=1}, varData []}}
      SectionFilter {allowedMaterialSlots []}
     }}
   )
   &self:ScratchModule(New Particle Attribute Reader=DataInterface() {propertyValues {EmitterBinding {bindingMode=Other, emitterName=Symbol}}}) {module=ScratchModule}
   &ParticleState()
   &ScaleSpriteSize(
    Non-Uniform Curve Scale=&RandomRangeVector2D(Minimum=<1 0.5>, Maximum=<1 1.5>)
    Non-Uniform Curve Sprite Scale=DataInterface() {propertyValues {
      XCurve {keys [{tangentMode=RCTM_User, time=0, value=1}, {tangentMode=RCTM_User, time=1, value=1}]}
      YCurve {keys [
        <at 0 0.5 RCIM_Cubic>
        {arriveTangent=1, interpMode=RCIM_Cubic, leaveTangent=1, time=0.9, value=2}
        {arriveTangent=-1.3057, interpMode=RCIM_Cubic, leaveTangent=-1.3057, tangentMode=RCTM_User, time=1, value=1.5}
       ]}
     }}
    Scale Sprite Size Mode=::NewEnumerator3
   )
   &ScaleColor(
    Scale Alpha=&FloatFromCurve(FloatCurve=DataInterface() {propertyValues {Curve {keys [<at 0 0>,<at 0.1 1>,<at 0.9 1>,<at 1 0>]}}})
    ScaleA=true
    ScaleRGB=true
   )
   &SolveForcesAndVelocity(Clamp Velocity=false, Limit Acceleration=false)
  ]
  rendererClasses [&NiagaraSpriteRendererProperties]
  renderers [{rendererClass=&NiagaraSpriteRendererProperties}]
 }
 Glow {
  emitterUpdateScript [
   &EmitterState(
    Inactive Response=::NewEnumerator1
    Life Cycle Mode=::NewEnumerator1
    Loop Behavior=::NewEnumerator1
    Loop Duration Mode=::NewEnumerator1
    UseLoopDelay=false
   )
   &SpawnBurst_Instantaneous(Use Spawn Probability=false, UseLoopCountLimit=false)
  ]
  particleSpawnScript [&InitializeParticle(
    Color=<rgba 3 2.5028 1.0113 0.05>
    Sprite Size Mode=::NewEnumerator3
    Uniform Sprite Size=40
    UsePositionOffset=false
    Write Material Random=true
   )]
  particleUpdateScript [
   &ParticleState(Kill Particles When Lifetime Has Elapsed=false)
   &ScaleColor(
    Scale RGB=&VectorFromFloat(Value=&FloatFromCurve(
      FloatCurve=DataInterface() {propertyValues {Curve {keys [<at 0 1>,<at 0.1 0.625>,<at 0.2 1.125>,<at 0.4 0.375>,<at 0.5 1.375>,<at 0.6 0.75>,<at 0.7 1.625>,<at 0.8 0.625>,<at 0.9 1.25>,<at 1 0.875>]}}}
      CurveIndex=&Multiply_Float(A=$System.LoopedAge, B=0.2)
     ))
    ScaleA=true
    ScaleRGB=true
   )
   /SetVariables(Particles.Position=&Add_Vector(A=<0 1 0>, B=$System.LogoOffset)) {isSetParameters=true, module=SetVariables_#4}
   &CameraOffset(Camera Offset Amount=10)
  ]
  rendererClasses [&NiagaraSpriteRendererProperties, &NiagaraLightRendererProperties]
  renderers [
   {rendererClass=&NiagaraSpriteRendererProperties}
   {rendererClass=&NiagaraLightRendererProperties, rendererIndex=1}
  ]
 }
 SpriteDash {
  emitterUpdateScript [
   &EmitterState(
    Inactive Response=::NewEnumerator1
    Life Cycle Mode=::NewEnumerator1
    Loop Behavior=::NewEnumerator1
    Loop Duration Mode=::NewEnumerator1
    UseLoopDelay=false
   )
   &SpawnRate(Spawn Probability=0.7, SpawnRate=4, Use Spawn Probability=true)
   &SpawnBurst_Instantaneous(Use Spawn Probability=false, UseLoopCountLimit=false) {enabled=false}
  ]
  particleSpawnScript [
   &InitializeParticle(
    Color=<rgba 1.1 0.9627 0 0.8>
    Lifetime=&RandomRangeFloat(Minimum=0.1, Maximum=0.3)
    Sprite Size Mode=::NewEnumerator3
    Uniform Sprite Size=&RandomRangeFloat(Minimum=45, Maximum=50)
    UsePositionOffset=false
    Write Material Random=true
   )
   &AddVelocity(
    Velocity=&RotateVectorByQuaternion(Vector To Rotate=<55 -1 100>, Quaternion=$System.AutoRotate)
    Velocity Speed Scale=0.1
   )
   &ShapeLocation(
    MaskSpawnGroup=false
    Non Uniform Scale=<1 0 0>
    Offset=<0 0 5>
    Offset Mode=::NewEnumerator0
    Override Seed=false
    Shape Origin:Vector3f=$System.LogoOffset
    Sphere Radius=4
    UseOwnerScale=true
   ) {enabled=false}
   &SpriteFacingAndAlignment(Sprite Alignment=<0 0 0>, Sprite Facing=<0 -1 0>, WriteAlignment=false, WriteFacing=true)
   /SetVariables(Particles.Position=&Add_Vector(A=<0 0 0>, B=$System.LogoOffset)) {isSetParameters=true, module=SetVariables_#4}
  ]
  particleUpdateScript [
   &ParticleState()
   &ScaleSpriteSize(
    Non-Uniform Curve Sprite Scale=DataInterface() {propertyValues {
      XCurve {keys [{tangentMode=RCTM_User, time=0, value=1}, {tangentMode=RCTM_User, time=1, value=1}]}
      YCurve {keys [{tangentMode=RCTM_User, time=0, value=0.8}, {tangentMode=RCTM_User, time=1, value=1}]}
     }}
    Scale Sprite Size Mode=::NewEnumerator3
   )
   &ScaleColor(
    Scale Alpha=&FloatFromCurve(FloatCurve=DataInterface() {propertyValues {Curve {keys [<at 0.7 1>,<at 1 0>]}}})
    ScaleA=true
    ScaleRGB=true
   )
   &CameraOffset(Camera Offset Amount=10)
   &SolveForcesAndVelocity(Clamp Velocity=false, Limit Acceleration=false)
  ]
  rendererClasses [&NiagaraSpriteRendererProperties]
  renderers [{rendererClass=&NiagaraSpriteRendererProperties}]
 }
 Symbol {
  emitterUpdateScript [
   &EmitterState(
    Inactive Response=::NewEnumerator1
    Life Cycle Mode=::NewEnumerator1
    Loop Behavior=::NewEnumerator1
    Loop Duration Mode=::NewEnumerator1
    UseLoopDelay=false
   )
   &SpawnBurst_Instantaneous(Use Spawn Probability=false, UseLoopCountLimit=false)
  ]
  particleSpawnScript [
   &InitializeParticle(Lifetime=5, Mesh Scale Mode=::NewEnumerator3, UsePositionOffset=false, Write Material Random=true)
   &DynamicMaterialParameters(
    Index 0 Param 1=$System.Random
    Index 0 Param 2=&MakeFloatFromVector(Vector=$Particles.Scale, Channel=::NewEnumerator0)
    Module.Param0WriteEnabled=true
    Module.Param1WriteEnabled=true
    Module.Param2WriteEnabled=false
    Module.Param3WriteEnabled=false
   )
   &InitialMeshOrientation(Mesh Orientation Mode=::NewEnumerator1, Rotation=<0.5 0 0>, Use Rotation Vector=true) {enabled=false}
  ]
  particleUpdateScript [
   &ParticleState(Kill Particles When Lifetime Has Elapsed=false, Let Infinitely Lived Particles Die When Emitter Deactivates=true)
   /SetVariables(Particles.MeshOrientation=$System.AutoRotate) {isSetParameters=true, module=SetVariables_#5}
   /SetVariables(Particles.Position=&Add_Vector(A=<0 0 0>, B=$System.LogoOffset)) {isSetParameters=true, module=SetVariables_#6}
   /SetVariables(Particles.ScanlinesOrientation:Vector4f=&MakeFloat4FromVectorAndFloat("Vector (RGB)"=&TransformVector(Vector=<0.2 -0.1 1>, Destination Space=::World), "Float (Alpha)"=1)) {isSetParameters=true}
  ]
  rendererClasses [&NiagaraMeshRendererProperties]
  renderers [{rendererClass=&NiagaraMeshRendererProperties, rendererIndex=0}]
 }
 Wings {
  emitterUpdateScript [
   &EmitterState(
    Inactive Response=::NewEnumerator1
    Life Cycle Mode=::NewEnumerator1
    Loop Behavior=::NewEnumerator1
    Loop Duration Mode=::NewEnumerator1
    UseLoopDelay=false
   )
   &SpawnBurst_Instantaneous(Use Spawn Probability=false, UseLoopCountLimit=false)
  ]
  particleSpawnScript [
   &InitializeParticle(Lifetime=5, Mesh Scale Mode=::NewEnumerator3, UsePositionOffset=false, Write Material Random=true)
   &DynamicMaterialParameters(
    Index 0 Param 1=$System.Random
    Index 0 Param 2=&MakeFloatFromVector(Vector=$Particles.Scale, Channel=::NewEnumerator0)
    Module.Param0WriteEnabled=true
    Module.Param1WriteEnabled=true
    Module.Param2WriteEnabled=false
    Module.Param3WriteEnabled=false
   )
  ]
  particleUpdateScript [
   &ParticleState(Kill Particles When Lifetime Has Elapsed=false, Let Infinitely Lived Particles Die When Emitter Deactivates=true)
   /SetVariables(Particles.MeshOrientation=$System.AutoRotate) {isSetParameters=true, module=SetVariables_#5}
   /SetVariables(Particles.Position=&Add_Vector(A=<0 0 0>, B=$System.LogoOffset)) {isSetParameters=true, module=SetVariables_#6}
   /SetVariables(Particles.ScanlinesOrientation:Vector4f=&MakeFloat4FromVectorAndFloat("Vector (RGB)"=&TransformVector(Vector=<0.2 -0.1 1>, Destination Space=::World), "Float (Alpha)"=1)) {isSetParameters=true}
  ]
  rendererClasses [&NiagaraMeshRendererProperties]
  renderers [
   {rendererClass=&NiagaraMeshRendererProperties, rendererIndex=0}
   {rendererClass=&NiagaraMeshRendererProperties, rendererIndex=1}
  ]
 }
}
summary {
 systemName=NS_BlueGlaze_Backpack_Idle
 emitters [
  {emitterName=Symbol, rendererClasses [&NiagaraMeshRendererProperties]}
  {emitterName=Glow, rendererClasses [&NiagaraSpriteRendererProperties, &NiagaraLightRendererProperties]}
  {emitterName=SpriteDash, rendererClasses [&NiagaraSpriteRendererProperties]}
  {emitterName=CRT_Yellow, rendererClasses [&NiagaraSpriteRendererProperties]}
  {emitterName=Wings, rendererClasses [&NiagaraMeshRendererProperties]}
  {emitterName=CRT_Rainbow, rendererClasses [&NiagaraSpriteRendererProperties]}
 ]
 userVariables [
  User.Backpack_DiffuseTex {
   type=Texture
   description=""
   defaultValue {
    struct=NiagaraExt_VariableValue_Object
    value {object=/BRCosmetics/Accessories/FORT_Backpacks/Backpack_M_MED_EerieGhost/FX/T_FX_EerieGhost_HologramSrc, objectClass=/Script/Engine.Texture}
   }
  }
  User.Backpack_Mat {
   type=MaterialInterface
   description=""
   defaultValue {
    struct=NiagaraExt_VariableValue_Object
    value {
     object=/BRCosmetics/Accessories/FORT_Backpacks/Backpack_M_MED_EerieGhost/FX/Material/MI_3DHB_EerieGhost
     objectClass=/Script/Engine.MaterialInterface
    }
   }
  }
  User.Backpack_NormalTex {
   type=Texture
   description=""
   defaultValue {
    struct=NiagaraExt_VariableValue_Object
    value {object=/BRCosmetics/Accessories/FORT_Backpacks/Backpack_M_MED_EerieGhost/FX/Texture/T__Normal, objectClass=&Texture}
   }
  }
  User.Backpack_WPOTex {
   type=Texture
   description=""
   defaultValue {
    struct=NiagaraExt_VariableValue_Object
    value {object=/BRCosmetics/Accessories/FORT_Backpacks/Backpack_M_MED_EerieGhost/FX/Texture/T__WPO, objectClass=&Texture}
   }
  }
  User.bIsFrontEnd {type=NiagaraBool}
  User.bIsFrontEndPreview {type=NiagaraBool}
  User.IsFrom3D {type=NiagaraFloat, defaultValue=0, description=""}
  User.Static Mesh {
   type=NiagaraDataInterfaceStaticMesh
   description=""
   defaultValue {
    struct=NiagaraExt_VariableValue_DataInterface
    value {
     dataInterface=&self:NiagaraDataInterfaceStaticMesh_0
     dataInterfaceClass=&NiagaraDataInterfaceStaticMesh
    }
   }
  }
 ]
}
